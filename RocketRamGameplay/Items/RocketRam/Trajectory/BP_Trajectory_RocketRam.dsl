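; Trajectory preview Blueprint for the Rocket Ram item, drawing a spline path with reticle meshes along it.
; Construction creates valid and invalid spline material instances and gathers eight reticle meshes.
; Each spline update spaces reticles at fixed distances, hiding those past the spline's end, and runs a one-time cosmetic setup of spline mesh material, scale and opacity.
; The validity event always passes true to the parent, and the obstructed-state event is empty.
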(event EventBeginPlay
  ;[0 1 0.008829] BeginPlay
  (|CallParentFunction))

(event Trajectory|EventSetTrajectorySpline
  (params (SplinePoints :type Array%20of%20Vectors) (SplineTangents :type Array%20of%20Vectors))
  (bind _splinecomponent (Variables|Trajectory|GetSplineComponent))
  (bind _distance_between_reticle_meshes_const (Variables|Default|GetDistance%20Between%20Reticle%20Meshes%20Const))
  (bind _distancetofirstreticlemeshconst (Variables|Default|GetDistanceToFirstReticleMeshConst))
  ;[0.7212 0 1] Event Set Trajectory Spline() Set the location and visibility of all of the reticle mesh components.
  (|CallParentFunction SplinePoints SplineTangents)
  (CallFunction|InitSplineMeshComponents)
  (for (_array_element _array_index) (Variables|Reticles|GetReticleMeshComponents)
    (Utilities|IsValid _array_element
      (:Is%20Valid
        (if (<= (+ (* _array_index _distance_between_reticle_meshes_const) _distancetofirstreticlemeshconst) (Spline|GetSplineLength _splinecomponent))
          ;;[0 1 0.7482] Set reticle location and unhide it
          (Transformation|SetWorldLocation _array_element (Spline|GetLocationatDistanceAlongSpline _splinecomponent (+ (* _array_index _distance_between_reticle_meshes_const) _distancetofirstreticlemeshconst) "World") false true)
          (Development|SetHiddeninGame _array_element false false)
          (else
            (Development|SetHiddeninGame _array_element true false)))))))

(event EventSetTrajectoryValid
  (params (bIsTrajectoryValid :type Boolean))
  (|CallParentFunction true))

(event Custom|SetTrajectoryObstructedState
  (params (RocketRamObstructedState :type Enum_RocketRamState%20Enum :default "NewEnumerator0")))

(fn DestroyAllReticlesMeshComponents
  (bind _reticlemeshcomponents (Variables|Reticles|GetReticleMeshComponents))
  (for _array_element _reticlemeshcomponents
    (Components|DestroyComponent _array_element))
  (Utilities|Array|Clear _reticlemeshcomponents))

(fn InitSplineMeshComponents
  (bind _splinemeshcomponents (Variables|Trajectory|GetSplineMeshComponents))
  (Utilities|FlowControl|DoOnce false
    (:Completed
      (Networking|CanExecuteCosmeticEvents
        (:True
          (Utilities|FlowControl|Sequence
            (:then_0
              ;[0.08218 1 0.00757] Set Material param for SplineMeshComponents[0]
              (if (Utilities|Array|IsValidIndex _splinemeshcomponents)
                (Rendering|Material|SetMaterial (Utilities|Array|Get(acopy) _splinemeshcomponents) 0 (Variables|VFX|GetSplineMID))))
            (:then_1
              ;[0 1 0.9566] Init all SplineMeshComponents[]
              (LocalVariables|SetSplineMeshLength (Utilities|Array|Length (Variables|Trajectory|GetSplineMeshComponents)))
              (for (_array_element _array_index) (Variables|Trajectory|GetSplineMeshComponents)
                (Utilities|IsValid _array_element
                  (:Is%20Valid
                    (SplineMesh|SetStartScale _array_element (Math|Vector2D|MakeVector2D (Math|Float|SelectFloat 3e+01 2e+01 (> _array_index 1)) (Math|Float|SelectFloat 3e+01 2e+01 (> _array_index 1))) false :!class /Script/Engine.SplineMeshComponent)
                    (SplineMesh|SetEndScale _array_element (35.0 35.0) true :!class /Script/Engine.SplineMeshComponent)
                    (Rendering|Material|SetScalarParameterValueonMaterials _array_element "Opacity" 0.9)))))))))))

(fn ConstructionScript
  ;[0.05805 0.4849 1] Create and store valid and invalid material instances.
  (bind _returnvalue (Rendering|Material|CreateDynamicMaterialInstance "/RocketRamGameplay/Items/RocketRam/Trajectory/M_Trajector_RocketRam.M_Trajector_RocketRam" "None" "None" :!class /Script/Engine.KismetMaterialLibrary))
  (Variables|VFX|SetSplineMID _returnvalue)
  (bind _returnvalue_2 (Rendering|Material|CreateDynamicMaterialInstance "/Game/Athena/Items/Consumables/TowerGrenade/M_Grenade_TrajectoryLine_01.M_Grenade_TrajectoryLine_01" "None" "None" :!class /Script/Engine.KismetMaterialLibrary))
  (Variables|VFX|SetInvalidSplineMID _returnvalue_2)
  (Variables|Reticles|SetReticleMeshComponents (Utilities|Array|MakeArray (Variables|Default|GetReticleMesh0) (Variables|Default|GetReticleMesh1) (Variables|Default|GetReticleMesh2) (Variables|Default|GetReticleMesh3) (Variables|Default|GetReticleMesh4) (Variables|Default|GetReticleMesh5) (Variables|Default|GetReticleMesh6) (Variables|Default|GetReticleMesh7))))
