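# Unlit translucent spline-mesh material that draws the Rocket Ram trajectory as a glowing band with a center line and repeating ticks.
# SphereMasks on UVs build the line and ticks, which combine with a soft background gradient to drive powered, depth-faded opacity.
# The line mask blends two emissive colors, and a time-looped spot sweeps a brighter color along U.
# EyeAdaptationInverse compensates the glow for exposure, and responsive AA is enabled.
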Material /RocketRamGameplay/Items/RocketRam/Trajectory/M_Trajector_RocketRam
graph {
 properties {
  bEnableResponsiveAA=true
  bUseTranslucencyVertexFog=false
  bUsedWithSplineMeshes=true
  blendMode=BLEND_Translucent
  pixelDepthOffsetMode=PDOM_Legacy
  shadingModel=MSM_Unlit
  translucencyPass=MTP_BeforeDOF
  usedShadingModels=Unlit
 }
 nodes [
  n1 = MaterialFunctionCall(Float2=TextureCoordinate()) {materialFunction=&BreakOutFloat2Components}
  n2 = ScalarParameter() {defaultValue=0.06, parameterName=RadiusCenterLine}
  n3 = TextureCoordinate()
  n4 = AppendVector(A=ScalarParameter() {defaultValue=80, parameterName=TickRepeat}, B=Constant() {r=1})
  n5 = Frac(None=ComponentMask(None=Multiply(A=@n3, B=@n4)) {r=true})
  n6 = Constant() {r=0.5}
  n7 = ScalarParameter() {defaultValue=0.06, parameterName=RadiusTicks}
  n8 = ScalarParameter() {defaultValue=0.5, parameterName=HardnessTicks}
  n9 = ComponentMask(None=@n3) {g=true}
  n10 = Constant() {r=0.5}
  n11 = Max(
   A=SphereMask(A=@n1.G, B=Constant() {r=0.5}, Radius=@n2, Hardness=ScalarParameter() {defaultValue=0.2, parameterName=HardnessCenterLine}) {attenuationRadius=0.5}
   B=Multiply(A=SphereMask(A=@n5, B=@n6, Radius=@n7, Hardness=@n8) {attenuationRadius=0.5}, B=SphereMask(A=@n9, B=@n10, Radius=@n10) {hardnessPercent=0.3})
  )
  n12 = Constant() {r=0.5}
  n13 = Multiply(
   A=SphereMask(A=@n1.G, B=@n12, Radius=@n12) {attenuationRadius=0.5, hardnessPercent=0}
   B=ScalarParameter() {defaultValue=0.9, parameterName=BGOpacity}
  )
  DepthFade(
   Opacity=Power(Base=Max(A=@n11, B=@n13), Exp=ScalarParameter() {defaultValue=3, parameterName=OpacityExponent})
   FadeDistance=ScalarParameter() {defaultValue=200, parameterName=DepthFadeDIstance}
  )
  n14 = VectorParameter() {parameterName=EmissiveColorBG, defaultValue=<rgba 2 0.2503 0.0042 1>}
  n15 = VectorParameter() {parameterName=EmissiveColor, defaultValue=<rgba 4 2.2977 0.0592 1>}
  n16 = MaterialFunctionCall!(Time, Divisor, Mod) {materialFunction=/Game/Effects/Fort_Effects/Materials/_Functions/MF_ModuloTime}
  n17 = Subtract(A=TextureCoordinate(), B=AppendVector(A=ConstantBiasScale(None=@n16.Clamped 0-1) {bias=-0.5, scale=2}, B=Constant()))
  n18 = Constant() {r=0.5}
  n19 = Multiply(B=@n2) {constA=4}
  EyeAdaptationInverse(LightValueInput=LinearInterpolate(
    A=LinearInterpolate(A=@n14.RGB, B=@n15.RGB, Alpha=@n11)
    B=VectorParameter() {parameterName=EmissiveSweep, defaultValue=<rgba 10 8.6095 0.1143 1>}.RGB
    Alpha=SphereMask(A=@n17, B=@n18, Radius=@n19) {attenuationRadius=0.1, hardnessPercent=0.5}
   ))
 ]
}
