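# Translucent unlit spline-mesh material for the backside of the Rocket Ram item's trajectory visual.
# An emissive color parameter passes through eye-adaptation inverse, fully desaturated where pixel depth exceeds scene depth.
# Opacity comes from a sphere mask on the V texture coordinate around a fixed center, forming a soft band shaped by radius, exponent and opacity parameters.
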Material /RocketRamGameplay/Items/RocketRam/Trajectory/M_Trajector_RocketRam_Backside
graph {
 properties {
  bEnableResponsiveAA=true
  bUseTranslucencyVertexFog=false
  bUsedWithSplineMeshes=true
  blendMode=BLEND_Translucent
  pixelDepthOffsetMode=PDOM_Legacy
  shadingModel=MSM_Unlit
  translucencyPass=MTP_BeforeDOF
  usedShadingModels=Unlit
 }
 nodes [
  n1 = Sign(None=Subtract(A=PixelDepth().R, B=SceneDepth().R))
  EyeAdaptationInverse(LightValueInput=Desaturation(None=VectorParameter() {parameterName=EmissiveColor, defaultValue=<rgba 2 1.0491 0.0539 1>}.RGB, Fraction=Saturate(None=@n1)))
  n2 = MaterialFunctionCall(Float2=TextureCoordinate()) {materialFunction=&BreakOutFloat2Components}
  n3 = Constant() {r=0.63}
  n4 = ScalarParameter() {defaultValue=0.5, parameterName=Radius}
  Multiply(
   A=Power(
    Base=SphereMask(A=@n2.G, B=@n3, Radius=@n4) {attenuationRadius=0.5, hardnessPercent=0}
    Exp=ScalarParameter() {defaultValue=3, parameterName=BGExponent}
   )
   B=ScalarParameter() {defaultValue=1, parameterName=Opacity}
  )
 ]
}
