Material /RocketRamGameplay/Items/RocketRam/Trajectory/M_Trajectory_RocketRam_Reticle
graph {
 properties {
  bDisableDepthTest=true
  bEnableResponsiveAA=true
  bUseTranslucencyVertexFog=false
  blendMode=BLEND_Translucent
  pixelDepthOffsetMode=PDOM_Legacy
  shadingModel=MSM_Unlit
  translucencyPass=MTP_BeforeDOF
  usedShadingModels=Unlit
 }
 nodes [
  LinearInterpolate(
   A=VectorParameter() {parameterName=ColorLow, defaultValue=<rgba 2 0.1254 0 1>}.RGB
   B=VectorParameter() {parameterName=Color, defaultValue=<rgba 4 1.4949 0 1>}.RGB
   Alpha=TextureSample() {texture=&/T_RocketRam_Reticle}.R
  )
  MaterialFunctionCall!(Center Location, 0-1 UVs, XY scale=ScalarParameter() {defaultValue=25, parameterName=ReticleSize}, OverridePivot, "Normals (Optional)") {materialFunction=&Sprite}
  n1 = TextureCoordinate()
  n2 = MaterialFunctionCall(Float2=Abs(None=ConstantBiasScale(None=@n1) {bias=-0.5, scale=2})) {materialFunction=&BreakOutFloat2Components}
  n3 = Max(A=@n2.R, B=@n2.R)
  n4 = ScalarParameter() {defaultValue=0.7, parameterName=Center}
  Multiply(
   A=SphereMask(A=@n3, B=@n4, Radius=ScalarParameter() {defaultValue=0.3, parameterName=RadiusBG}, Hardness=Constant())
   B=ScalarParameter() {defaultValue=0.6, parameterName=BGScale}
  )
  SphereMask(
   A=@n3
   B=@n4
   Radius=ScalarParameter() {defaultValue=0.04, parameterName=Radius}
   Hardness=ScalarParameter() {defaultValue=0.1, parameterName=Hardness}
  )
 ]
}
